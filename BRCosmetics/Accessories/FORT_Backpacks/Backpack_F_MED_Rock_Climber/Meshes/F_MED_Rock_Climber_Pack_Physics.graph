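# Physics asset for the Rock Climber backpack cosmetic, defining box and capsule bodies that let its hanging gear swing.
# Scope, ice axe, two two-link rope loops and a two-link carabiner chain constrain back to dyn_master, with the ice axe held nearly rigid.
# Five rope folds swing from dyn_rope_knot, which has no constraint to dyn_master.
# Each joint sets swing and twist limits, with twist kept tight.
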PhysicsAsset /BRCosmetics/Accessories/FORT_Backpacks/Backpack_F_MED_Rock_Climber/Meshes/F_MED_Rock_Climber_Pack_Physics
bodies {
 dyn_carabiner [{shapeType=Box, extentX=3.5763, extentY=1.6404, extentZ=5.8123, center=<3.013 -0.1508 0.1216>, rotation=<rot 0 87.373 -88.9285>}]
 dyn_iceaxe_01 [{shapeType=Capsule, length=2.6019, radius=1.4646, center=<0.3379 0.0117 -0.0036>, rotation=<rot 0 91.9814 -90.6124>}]
 dyn_master [{shapeType=Capsule, length=3, radius=6.375, center=<1.4158 1.9409 0.7754>, rotation=<rot -85.3956 0.0003 -90.0887>}]
 dyn_ropcarabiner [{shapeType=Box, extentX=2.5971, extentY=1.9381, extentZ=0.8403, center=<1.378 0 -0.0001>}]
 dyn_rope_knot [{shapeType=Box, extentX=1, extentY=1.5, extentZ=1.5}]
 dyn_ropefold_01 [{shapeType=Box, extentX=7, extentY=3.9513, extentZ=1, center=<2.8205 -1.1879 0.9582>, rotation=<rot 10.6898 -19.0628 30.1357>}]
 dyn_ropefold_02 [{shapeType=Box, extentX=7, extentY=3.9513, extentZ=1, center=<2.9626 -1.5034 0.5633>, rotation=<rot 5.9021 -22.0112 26.3338>}]
 dyn_ropefold_03 [{shapeType=Box, extentX=7, extentY=3.9513, extentZ=1, center=<2.6659 -1.5394 0.2427>, rotation=<rot 3.4843 -23.2087 26.2356>}]
 dyn_ropefold_04 [{shapeType=Box, extentX=4, extentY=3.2013, extentZ=0.75, center=<1.831 -1.4881 -0.8162>, rotation=<rot -2.6101 -14.1959 -23.315>}]
 dyn_ropefold_05 [{shapeType=Box, extentX=4, extentY=3.2013, extentZ=0.75, center=<1.6051 -1.6757 -0.5753>, rotation=<rot 3.0807 -16.6493 -23.3231>}]
 dyn_ropeloop_01 [{shapeType=Box, extentX=2.8517, extentY=1.2907, extentZ=0.9492, center=<1.2856 0 0>}]
 dyn_ropeloop_02 [{shapeType=Box, extentX=3.0071, extentY=1.4561, extentZ=1.4972, center=<1.1037 0 0>}]
 dyn_ropeloop_03 [{shapeType=Capsule, length=1.3769, radius=0.5579, center=<1.0221 0 -0.145>, rotation=<rot 81.929 0 0>}]
 dyn_ropeloop_04 [{shapeType=Box, extentX=2.8517, extentY=1.2907, extentZ=0.9492, center=<1.2856 0 0>}]
 dyn_scope [{shapeType=Capsule, length=3.8519, radius=1.2146, rotation=<rot 90 0 0>}]
}
constraints [
 {bone1Name=dyn_scope, bone2Name=dyn_master, swing1LimitDegrees=20, swing2LimitDegrees=15, twistLimitDegrees=3}
 {bone1Name=dyn_iceaxe_01, bone2Name=dyn_master, swing1LimitDegrees=2, swing2LimitDegrees=2, twistLimitDegrees=2}
 {bone1Name=dyn_ropefold_01, bone2Name=dyn_rope_knot, swing1LimitDegrees=45, swing2LimitDegrees=45, twistLimitDegrees=8}
 {bone1Name=dyn_ropefold_02, bone2Name=dyn_rope_knot, swing1LimitDegrees=35, swing2LimitDegrees=35, twistLimitDegrees=8}
 {bone1Name=dyn_ropefold_03, bone2Name=dyn_rope_knot, swing1LimitDegrees=40, swing2LimitDegrees=40, twistLimitDegrees=8}
 {bone1Name=dyn_ropefold_04, bone2Name=dyn_rope_knot, swing1LimitDegrees=50, swing2LimitDegrees=50, twistLimitDegrees=10}
 {bone1Name=dyn_ropefold_05, bone2Name=dyn_rope_knot, swing1LimitDegrees=50, swing2LimitDegrees=50, twistLimitDegrees=10}
 {bone1Name=dyn_ropeloop_01, bone2Name=dyn_master, swing1LimitDegrees=30, swing2LimitDegrees=30, twistLimitDegrees=5}
 {bone1Name=dyn_ropeloop_02, bone2Name=dyn_ropeloop_01, swing1LimitDegrees=35, swing2LimitDegrees=35, twistLimitDegrees=10}
 {bone1Name=dyn_ropeloop_04, bone2Name=dyn_ropeloop_03, swing1LimitDegrees=35, swing2LimitDegrees=35, twistLimitDegrees=5}
 {bone1Name=dyn_carabiner, bone2Name=dyn_ropcarabiner, swing1LimitDegrees=40, swing2LimitDegrees=50, twistLimitDegrees=10}
 {bone1Name=dyn_ropcarabiner, bone2Name=dyn_master, swing1LimitDegrees=40, swing2LimitDegrees=50, twistLimitDegrees=10}
 {bone1Name=dyn_ropeloop_03, bone2Name=dyn_master, swing1LimitDegrees=35, swing2LimitDegrees=35, twistLimitDegrees=5}
]
