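# Animation Blueprint for the Rock Climber backpack cosmetic, adding physics-driven secondary motion to its skeleton.
# The input pose passes through two constraint nodes that tie the dyn_iceaxe_01 and dyn_scope bones to the pack bone, then into a RigidBody node.
# That node simulates the pack's physics asset in root-bone space with overridden gravity, scaled-down component motion and translation limits clamped to the reference pose.
# The event graph holds no logic.
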AnimBlueprint /BRCosmetics/Accessories/FORT_Backpacks/Backpack_F_MED_Rock_Climber/Meshes/F_MED_Rock_Climber_Pack_AnimBp
target_skeleton=&/F_MED_Rock_Climber_Pack_Skeleton
graphs [
 AnimGraph {nodes [
   AnimGraphNode_SubInput_0 = Input Pose()
   n1 = "Constraint - dyn_iceaxe_01 (P)"!(ComponentPose=Local To Component(LocalPose=@AnimGraphNode_SubInput_0), Alpha=1, ConstraintWeights_0=1, ConstraintWeights_1:float=0) -> (Pose) {
    alphaBoolBlend {blendOption=Linear}
    boneToModify {boneName=dyn_iceaxe_01}
    constraintSetup [{targetBone {boneName=pack}}]
    constraintWeights [1,0]
   }
   n2 = "Constraint - dyn_scope (P)"!(ComponentPose=@n1, Alpha=1, ConstraintWeights_0=1, ConstraintWeights_1:float=0) -> (Pose) {
    alphaBoolBlend {blendOption=Linear}
    boneToModify {boneName=dyn_scope}
    constraintSetup [{targetBone {boneName=pack}}]
    constraintWeights [1,0]
   }
   n3 = RigidBody(ComponentPose=@n2) {
    bClampLinearTranslationLimitToRefPose=true
    bOverrideWorldGravity=true
    overridePhysicsAsset=&/F_MED_Rock_Climber_Pack_Physics
    simulationSpace=BaseBoneSpace
    alphaBoolBlend {blendOption=Linear}
    baseBoneRef {boneName=root}
    componentAppliedLinearAccClamp=<4000 4000 4000>
    componentLinearAccScale=<0.25 0.25 0.25>
    componentLinearVelScale=<0.25 0.25 0.25>
    overrideWorldGravity {z=-980}
   }
   Output Pose(Result=Component To Local(ComponentPose=@n3)) {layerGroup=None}
  ]}
 EventGraph {nodes [
   TryGetPawnOwner()
   Event BlueprintUpdateAnimation()
  ]}
]
variables []
